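# Unlit translucent Niagara ribbon material for the Cupid Evil pickaxe trail.
# Panned, bendable lava noise is eroded by a UV-shaped mask, driven by a dynamic Erosion parameter and pinched at the base, to form opacity with depth fade.
# The noise and erosion also blend four colors, mixed with particle color for emissive.
# A normal-direction offset bulges the ribbon center, and the dynamic U offset drives both noise UV axes.
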Material /BRCosmetics/Weapons/FORT_Melee/Pickaxe_CupidEvil/FX/M_Pickaxe_CupidEvil_Trail
graph {
 properties {bUsedWithNiagaraRibbons=true, blendMode=BLEND_Translucent, pixelDepthOffsetMode=PDOM_Legacy, shadingModel=MSM_Unlit, usedShadingModels=Unlit}
 nodes [
  n1 = NamedRerouteUsage()
  n2 = OneMinus(None=@n1)
  n3 = ScalarParameter() {defaultValue=1, parameterName=Color Lerp Exponent}
  n4 = MaterialFunctionCall(
   1=VectorParameter() {parameterName=Color 1, defaultValue=<rgba 11 1.65 0 0>}.RGB
   2=VectorParameter() {parameterName=Color 2, defaultValue=<rgba 4 0.2 0 0>}.RGB
   3=VectorParameter() {parameterName=Color 3, defaultValue=<rgba 0.09 0.0077 0.0063 0>}.RGB
   A=OneMinus(None=Power(Base=@n2, Exp=@n3))
  ) {materialFunction=&Lerp_Multiple_Float3}
  n5 = VectorParameter() {parameterName=Color 4, defaultValue=<rgba 0.04 0.0034 0.0028 0.666>}
  n6 = NamedRerouteUsage()
  n7 = ScalarParameter() {defaultValue=1, parameterName=Erosion Color Softness}
  n8 = Power(
   Base=OneMinus(None=MaterialFunctionCall!(Gradient=@n1, Erosion Amount=@n6, Softness=@n7) {materialFunction=/BRCosmetics/Athena/Effects/Particles/Skydiving/NS_Contrail_DarkAndLight/Functions/MF_Erosion}.Increased Range)
   Exp=@n5.A
  )
  n9 = ParticleColor()
  n10 = SquareRoot(None=@n1)
  n11 = MaterialFunctionCall(
   Gradient=@n10
   Erosion Amount=Add(A=@n6, B=ScalarParameter() {defaultValue=0.1325, parameterName=Erosion Opacity Softness}) {constB=0.0666}
   Softness=ScalarParameter() {defaultValue=1, parameterName=Edge Color Softness}
  ) {materialFunction=&MF_Erosion}
  MaterialFunctionCall(Emissive=LinearInterpolate(A=ParticleColor().RGB, B=LinearInterpolate(A=@n4.Lerp 3 Inputs, B=@n5.RGB, Alpha=@n8), Alpha=Multiply(A=@n9.A, B=@n11.Increased Range))) {materialFunction=&MF_FX_EmissiveAdjustments}
  n12 = ConstantBiasScale(None=ComponentMask(None=TextureCoordinate()) {g=true}) {bias=-0.5, scale=2}
  n13 = DotProduct(A=@n12, B=@n12)
  Multiply(A=Multiply(A=OneMinus(None=@n13), B=ScalarParameter() {defaultValue=6, parameterName=Center Bulge}), B=VertexNormalWS())
  n14 = TextureCoordinate()
  n15 = OneMinus(None=ComponentMask(None=@n14) {r=true})
  n16 = ComponentMask(None=@n14) {g=true}
  n17 = Divide(A=ComponentMask(None=@n14) {r=true}, B=ScalarParameter() {defaultValue=0.15, parameterName=Erosion Base Size})
  n18 = OneMinus(None=SquareRoot(None=Saturate(None=@n17)))
  n19 = Saturate(None=ScalarParameter() {defaultValue=1, parameterName=Erosion Base Pinch})
  n20 = Multiply(A=@n18, B=@n19)
  n21 = Multiply(A=Divide(A=Subtract(A=@n16) {constB=0.5}, B=OneMinus(None=@n20))) {constB=2}
  n22 = DotProduct(A=@n21, B=@n21)
  n23 = OneMinus(None=@n22)
  n24 = Multiply(A=@n18, B=OneMinus(None=@n19))
  n25 = Multiply(
   A=Saturate(None=DynamicParameter() {defaultValue=<rgba 0 0 0 0>, paramNames [UV Offset U, UV Offset V, "Erosion (0-1)", - unused -]}."Erosion (0-1)")
   B=Multiply(A=Multiply(A=@n15, B=@n23), B=OneMinus(None=@n24))
  )
  n26 = DepthFade(Opacity=@n25, FadeDistance=ScalarParameter() {defaultValue=22, parameterName=Depth Fade Distance})
  NamedRerouteDeclaration(None=OneMinus(None=QualitySwitch(Default=@n26, Low=@n25, Medium=@n25))) {name=Erosion Amount, nodeColor=<rgba 0.4766 0 1 1>}
  n27 = MaterialFunctionCall(Float2=TextureCoordinate()) {materialFunction=&BreakOutFloat2Components}
  n28 = NamedRerouteUsage()
  n29 = ScalarParameter() {parameterName=UV Length Bend Intensity}
  n30 = AppendVector(A=@n27.G, B=Add(A=@n27.R, B=Multiply(A=@n28, B=@n29)))
  n31 = ScalarParameter() {defaultValue=0.15, parameterName=UV Width Scale}
  n32 = ScalarParameter() {defaultValue=1.5, parameterName=UV Length Scale}
  n33 = Add(A=Multiply(A=Subtract(A=@n30) {constB=0.5}, B=AppendVector(A=@n31, B=@n32))) {constB=0.5}
  n34 = AppendVector(A=Constant(), B=ScalarParameter() {defaultValue=1, parameterName=UV Length Panner Speed})
  n35 = DynamicParameter() {defaultValue=<rgba 0 0 0 0>, paramNames [UV Offset U, UV Offset V, "Erosion (0-1)", - unused -]}
  n36 = TextureSample(UVs=Add(A=Panner(Coordinate=@n33, Speed=@n34) {bFractionalPart=true}, B=AppendVector(A=@n35.UV Offset U, B=@n35.UV Offset U))) {
   samplerType=SAMPLERTYPE_LinearGrayscale
   texture=/BRCosmetics/Weapons/FORT_Melee/Pickaxe_RenegadeRaider_Spark/FX/Materials/T_Pickaxe_RenegadeRaider_Spark_LavaNoise
  }
  NamedRerouteDeclaration(None=@n36.R) {name=Noise Texture, nodeColor=<rgba 0.3359 0 1 1>}
  NamedRerouteDeclaration(None=@n22) {name=UV Bend, nodeColor=<rgba 0 1 0.0391 1>}
  n37 = MaterialFunctionCall(Gradient=@n10, Erosion Amount=@n6, Softness=ScalarParameter() {defaultValue=0.25, parameterName=Softness}) {materialFunction=&MF_Erosion}
  n38 = DepthFade(Opacity=@n37.Increased Range, FadeDistance=ScalarParameter() {defaultValue=20, group=Depth Fade, parameterName=DepthFade})
  StaticSwitchParameter(
   True=FeatureLevelSwitch(Default=@n38, ES2=@n37.Increased Range, ES3_1=@n37.Increased Range, SM4_REMOVED=@n38, SM5=@n38, SM6=@n38)
   False=@n37.Increased Range
  ) {parameterName=HasDepthFade}
  VectorParameter() {parameterName=Color 5, defaultValue=<rgba 0.0333 0.014 0.0196 1>}
 ]
}
