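# Niagara swing trail for the CupidEvil pickaxe, layering fiery ribbons, dark smoke and hot orange-to-red embers.
# Four ribbon emitters use shared AnimTrail modules, with fire and smoke layers sampling User.SkeletalMesh sockets for offset and velocity.
# Sprite embers start after a brief age-gated delay, with one per-particle random correlating lifetime, speed and size.
# User.WeakPointHitFade drives only Trail_Screen, lifting its RGB from black and boosting alpha.
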NiagaraSystem /BRCosmetics/Weapons/FORT_Melee/Pickaxe_CupidEvil/FX/NS_Pickaxe_CupidEvil_Trail
dependencies {
 usedDataInterfaces [&NiagaraDataInterfaceCurve, &NiagaraDataInterfaceColorCurve]
 usedDynamicInputs [&Frac_Float, &Multiply_Float]
 usedModules [
  &self:SystemSpawnScript.SystemScriptSource.SystemScriptGraph.NiagaraNodeAssignment_1.SetVariables_#1
  &SystemState
 ]
}
emitters {
 Embers {
  emitterUpdateScript [
   &EmitterState()
   &SpawnRate(
    Spawn Probability=&FloatFromCurve(FloatCurve=DataInterface() {propertyValues {Curve {keys [<at 0.1 0>,<at 0.13 1>]}}}, CurveIndex=$Emitter.Age)
    SpawnRate=222
    Use Spawn Probability=true
   )
  ]
  particleSpawnScript [
   /SetVariables(Particles.PerParticleRandom:NiagaraFloat=&RandomRangeFloat(Minimum=0, Maximum=1)) {isSetParameters=true}
   &InitializeParticle(
    Color Mode=::NewEnumerator0
    Lifetime=&Lerp_Float(A=0.777, B=0.222, Alpha=$Particles.PerParticleRandom)
    Sprite Size=&Multiply_Vector2D(
     A=<1 1>
     B=&MakeVector2D(
      X=&FloatFromCurve(
       FloatCurve=DataInterface() {propertyValues {Curve {keys [<at 0 1>,<at 0.8 2 RCIM_Constant>,<at 0.95 3 RCIM_Constant>]}}}
       CurveIndex=$Particles.PerParticleRandom
      )
      Y=&FloatFromCurve(
       FloatCurve=DataInterface() {propertyValues {Curve {keys [<at 0 1>,<at 0.8 2 RCIM_Constant>,<at 0.95 3>]}}}
       CurveIndex=$Particles.PerParticleRandom
      )
     )
    )
    Sprite Size Mode=::NewEnumerator4
    UsePositionOffset=false
    Write Material Random=true
   )
   &AddVelocity(
    Module.UseConeFalloff=false
    Module.UseVelDistribution=false
    Override Seed=false
    Velocity Mode=::NewEnumerator2
    Velocity Speed=&Lerp_Float(A=-33, B=666, Alpha=$Particles.PerParticleRandom)
   )
   &InheritVelocity(Speed Limit=true, UseSpeedThreshold=false)
   &ShapeLocation(
    MaskSpawnGroup=false
    Offset=<0 0 40>
    Offset Coordinate Space=::Local
    Offset Mode=::NewEnumerator0
    Override Seed=false
    Sphere Radius=11
    UseOwnerScale=true
   )
   &ApplyInitialForces()
  ]
  particleUpdateScript [
   &ParticleState()
   &Drag(
    Drag=&FloatFromCurve(
     FloatCurve=DataInterface() {propertyValues {Curve {keys [<at 0 0.5>,<at 0.5 0.5>,<at 1 1>]}}}
     Scale Curve=&RandomRangeFloat(Minimum=3, Maximum=22)
    )
    Use Linear Drag=true
    Use Rotational Drag=false
   )
   &CurlNoiseForce(Noise Frequency=22, Noise Strength=666, Pan Noise=false, Random Seed=33) {module=CurlNoiseForce001}
   &GravityForce(Gravity=<0 0 333>)
   &SolveForcesAndVelocity(Clamp Velocity=false, Limit Acceleration=false)
   &ScaleColor(
    Scale RGB=&MakeVectorFromLinearColorRGB(Color=&ColorFromCurve(ColorCurve=DataInterface() {propertyValues {
        AlphaCurve {keys [<at 0 1 RCIM_Cubic>]}
        BlueCurve {keys [<at 0 0 RCIM_Cubic>,<at 0.4873 0>,<at 1 0 RCIM_Cubic>]}
        GreenCurve {keys [<at 0 3.3 RCIM_Cubic>,<at 0.4873 0.0333>,<at 1 0 RCIM_Cubic>]}
        RedCurve {keys [<at 0 22 RCIM_Cubic>,<at 0.4873 1>,<at 1 1 RCIM_Cubic>]}
       }}))
    ScaleA=true
    ScaleRGB=true
   )
   &ScaleSpriteSize(
    Uniform Curve Scale=&MakeFloatFromVector(Vector=$Engine.Owner.Scale, Channel=::NewEnumerator0)
    Uniform Curve Sprite Scale=DataInterface() {propertyValues {Curve {keys [
        {tangentMode=RCTM_User, time=0, value=1}
        {arriveTangent=-0.0405, interpMode=RCIM_Cubic, leaveTangent=-0.0405, tangentMode=RCTM_User, time=0.7009, value=1}
        {arriveTangent=-8.0089, interpMode=RCIM_Cubic, leaveTangent=-8.0089, tangentMode=RCTM_User, time=1, value=0}
       ]}}}
   )
   &ScaleSpriteSizeBySpeed(
    Max Scale Factor=<1.666 13>
    Min Scale Factor=<1.333 1.333>
    Sample Scale Factor By Curve=true
    Scale Factor Curve=DataInterface() {propertyValues {Curve {keys [<at 0 0>, {arriveTangent=1.4989, interpMode=RCIM_Cubic, tangentMode=RCTM_Break, time=1, value=1}]}}}
    Velocity Threshold=999
   )
  ]
  rendererClasses [&NiagaraSpriteRendererProperties]
  renderers [{rendererClass=&NiagaraSpriteRendererProperties}]
 }
 Trail_Fire {
  emitterUpdateScript [
   &EmitterState(Life Cycle Mode=::NewEnumerator1, Loop Behavior=::NewEnumerator1, Loop Duration=0.27, UseLoopDelay=true)
   &FN_SpawnAnimTrail_v2()
  ]
  particleSpawnScript [
   &SampleSkeletalMesh(
    AdditionalYawPitchRoll=false
    Bone / Socket Index:NiagaraInt32=&Add_Integer(A=$Particles.SpawnGroup, B=0)
    Mesh Sampling Type:ENiagaraSkelSamplingModeFull=::NewEnumerator5
    Skeletal Mesh:NiagaraDataInterfaceSkeletalMesh=$User.SkeletalMesh
    Socket Sampling Mode:ENiagaraSocketSamplingMode=::NewEnumerator5
   )
   /Game/Effects/Niagara/Modules/AnimTrail/FN_InitializeAnimTrail/FN_InitializeAnimTrailBoneIndecies(First Bone or Socket Index=1)
   /Game/Effects/Niagara/Modules/AnimTrail/FN_InitializeAnimTrail/FN_InitializeAnimTrail(
    AnimTrailWidth=1
    Color=<rgba 0 0 0 4>
    Lifetime=0.225
    Position Offset:Vector3f=&Multiply_VectorByFloat(
     Vector=&NormalizeVector(
      Vector To Normalize=$Particles.SampleSkeletalMesh.SampledVelocity
      Safe Normalize=true
      Safe Fallback Threshold=0
      Safe Fallback Vector=<0 0 0.001>
     )
     Float=15
    )
   )
   &SystemLocation(MaskSpawnGroup=false)
   &AddVelocity(
    Rotation Coordinate Space=::Simulation
    Velocity=&RotateVectorByQuaternion(Vector To Rotate=<1 0 2>, Quaternion=$Particles.SampleSkeletalMesh.SampledRotation)
    Velocity Speed Scale=100
   ) {module=AddVelocity001}
  ]
  particleUpdateScript [
   &FN_UpdateAnimTrail_v2()
   &ScaleRibbonWidth(Ribbon Width Scale=&FloatFromCurve(FloatCurve=DataInterface() {propertyValues {Curve {keys [
         {tangentMode=RCTM_User, time=0, value=1}
         {arriveTangent=-0.0405, interpMode=RCIM_Cubic, leaveTangent=-0.0405, tangentMode=RCTM_User, time=0.7009, value=1}
         {arriveTangent=-8.0089, interpMode=RCIM_Cubic, leaveTangent=-8.0089, tangentMode=RCTM_User, time=1, value=0}
        ]}}})) {module=ScaleRibbonWidth001}
   &GravityForce(Gravity=<0 0 0>)
   &SolveForcesAndVelocity(Clamp Velocity=false, Limit Acceleration=false)
   &DynamicMaterialParameters(
    Index 0 Param 1=&RandomRangeFloat(Minimum=-0.25, Maximum=0.25)
    Index 0 Param 2=&FloatFromCurve(FloatCurve=DataInterface() {propertyValues {Curve {keys [<at 0 1>,<at 1 0.55>]}}})
    Module.Param0WriteEnabled=true
    Module.Param1WriteEnabled=true
    Module.Param2WriteEnabled=false
    Module.Param3WriteEnabled=false
   )
   &Color(Color=&ColorFromCurve(ColorCurve=DataInterface() {propertyValues {
       AlphaCurve {keys [
         {arriveTangent=-3.2415, interpMode=RCIM_Cubic, leaveTangent=-3.2415, tangentMode=RCTM_User, time=0.0615, value=1.5}
         <at 0.9912 0 RCIM_Cubic>
        ]}
       BlueCurve {keys [
         <at 0.0075 0>
         {arriveTangent=0.0024, interpMode=RCIM_Cubic, leaveTangent=0.0024, time=0.0876, value=0}
         {arriveTangent=0.0024, interpMode=RCIM_Cubic, leaveTangent=0.0024, time=0.2292, value=0}
         <at 0.4807 0.0011>
        ]}
       GreenCurve {keys [
         <at 0.0075 0.9302>
         {arriveTangent=-1.9657, interpMode=RCIM_Cubic, leaveTangent=-1.9657, time=0.0876, value=0.25}
         {arriveTangent=-1.9657, interpMode=RCIM_Cubic, leaveTangent=-1.9657, time=0.2292, value=0.0417}
         <at 0.4807 0>
        ]}
       RedCurve {keys [
         <at 0.0075 3>
         {arriveTangent=-6.2843, interpMode=RCIM_Cubic, leaveTangent=-6.2843, time=0.0876, value=3}
         {arriveTangent=-6.2843, interpMode=RCIM_Cubic, leaveTangent=-6.2843, time=0.2292, value=0.5}
         <at 0.4807 0.026>
        ]}
      }}))
   /SetVariables(Particles.Color=&MultiplyLinearColor(Linear Color=$Particles.Color, Scale Factor=&MakeVector4(X=1, Y=1, Z=1, W=1))) {isSetParameters=true}
  ]
  rendererClasses [&NiagaraRibbonRendererProperties]
  renderers [{rendererClass=&NiagaraRibbonRendererProperties}]
 }
 Trail_Fire001 {
  emitterUpdateScript [
   &EmitterState(Life Cycle Mode=::NewEnumerator1, Loop Behavior=::NewEnumerator1, Loop Duration=0.17, UseLoopDelay=true)
   &FN_SpawnAnimTrail_v2()
  ]
  particleSpawnScript [
   &SampleSkeletalMesh(
    AdditionalYawPitchRoll=false
    Bone / Socket Index=3
    Mesh Sampling Type:ENiagaraSkelSamplingModeFull=::NewEnumerator5
    Skeletal Mesh:NiagaraDataInterfaceSkeletalMesh=$User.SkeletalMesh
    Socket Sampling Mode:ENiagaraSocketSamplingMode=::NewEnumerator5
   )
   &FN_InitializeAnimTrailBoneIndecies(First Bone or Socket Index=2, Second Bone or Socket Index=2)
   &FN_InitializeAnimTrail(
    AnimTrailWidth=45
    Lifetime=0.25
    Position Offset:Vector3f=&Multiply_VectorByFloat(
     Vector=&NormalizeVector(
      Vector To Normalize=$Particles.SampleSkeletalMesh.SampledVelocity
      Safe Normalize=true
      Safe Fallback Threshold=0
      Safe Fallback Vector=<0 0 0.001>
     )
     Float=0
    )
   )
   &AddVelocity(
    Rotation Coordinate Space=::Simulation
    Velocity=&RotateVectorByQuaternion(Vector To Rotate=<1 0 0>, Quaternion=$Particles.SampleSkeletalMesh.SampledRotation)
    Velocity Speed Scale=100
   ) {module=AddVelocity001}
   &CameraOffset()
   &DynamicMaterialParameters(
    Index 0 Param 1=$System.RandomFloatA
    Index 0 Param 2=$System.RandomFloatB
    Module.Param0WriteEnabled=true
    Module.Param1WriteEnabled=true
    Module.Param2WriteEnabled=false
    Module.Param3WriteEnabled=false
   )
  ]
  particleUpdateScript [
   &FN_UpdateAnimTrail_v2()
   &ScaleRibbonWidth(Ribbon Width Scale=&FloatFromCurve(FloatCurve=DataInterface() {propertyValues {Curve {keys [
         {interpMode=RCIM_Cubic, tangentMode=RCTM_User, time=0, value=0.5}
         {arriveTangent=1.3038, interpMode=RCIM_Cubic, leaveTangent=1.3038, tangentMode=RCTM_User, time=1, value=1}
        ]}}})) {module=ScaleRibbonWidth001}
   &CurlNoiseForce(Noise Frequency=22, Noise Strength=666, Pan Noise=false, Random Seed=33)
   &GravityForce(Gravity=<0 0 0>)
   &SolveForcesAndVelocity(Clamp Velocity=false, Limit Acceleration=false)
   &Color(Color=&ColorFromCurve(ColorCurve=DataInterface() {propertyValues {
       AlphaCurve {keys [
         {arriveTangent=-3.2415, interpMode=RCIM_Cubic, leaveTangent=-3.2415, tangentMode=RCTM_User, time=0.0037, value=1.5}
         {arriveTangentWeight=0.326, interpMode=RCIM_Cubic, tangentMode=RCTM_User, time=0.5608, value=0}
        ]}
       BlueCurve {keys [<at 0.0149 0>, <at 0.0689 0 RCIM_Cubic>, {interpMode=RCIM_Cubic, leaveTangent=0.0265, time=0.123, value=0}, <at 0.1882 0.0091>]}
       GreenCurve {keys [
         <at 0.0149 0.3101>
         {arriveTangent=-2.6767, interpMode=RCIM_Cubic, leaveTangent=-2.6767, time=0.0689, value=0.0833}
         {interpMode=RCIM_Cubic, leaveTangent=-0.8809, time=0.123, value=0.0208}
         <at 0.1882 0.0065>
        ]}
       RedCurve {keys [
         <at 0.0149 1>
         {arriveTangent=-6.9405, interpMode=RCIM_Cubic, leaveTangent=-6.9405, time=0.0689, value=1}
         {interpMode=RCIM_Cubic, leaveTangent=-2.8559, time=0.123, value=0.25}
         <at 0.1882 0.0156>
        ]}
      }}))
   &DynamicMaterialParameters(
    Index 0 Param 3=&FloatFromCurve(FloatCurve=DataInterface() {propertyValues {Curve {keys [
         {arriveTangent=-1.8694, interpMode=RCIM_Cubic, leaveTangent=-1.8694, tangentMode=RCTM_User, time=0, value=1}
         {interpMode=RCIM_Cubic, tangentMode=RCTM_User, time=1, value=0}
        ]}}})
    Module.Index0Param0Write=false
    Module.Index0Param1Write=false
    Module.Index0Param2Write=true
    Module.Index0Param3Write=false
   ) {module=DynamicMaterialParameters001}
  ]
  rendererClasses [&NiagaraRibbonRendererProperties]
  renderers [{rendererClass=&NiagaraRibbonRendererProperties}]
 }
 Trail_Screen {
  emitterUpdateScript [
   &EmitterState(Life Cycle Mode=::NewEnumerator1, Loop Behavior=::NewEnumerator1, Loop Duration=0.27, UseLoopDelay=true)
   &FN_SpawnAnimTrail_v2(AnimTrailWidthMode=::FromSecond)
  ]
  particleSpawnScript [
   &FN_InitializeAnimTrailBoneIndecies(First Bone or Socket Index=1)
   &FN_InitializeAnimTrail(AnimTrailWidth=0.75, Color=<rgba 0 0 0 3>, Lifetime=0.225)
   &SystemLocation(MaskSpawnGroup=false)
  ]
  particleUpdateScript [
   &FN_UpdateAnimTrail_v2()
   &GravityForce(Gravity=<0 0 300>)
   &SolveForcesAndVelocity(Clamp Velocity=false, Limit Acceleration=false)
   &ScaleRibbonWidth(Ribbon Width Scale=&FloatFromCurve(FloatCurve=DataInterface() {propertyValues {Curve {keys [
         {arriveTangent=-1.2579, interpMode=RCIM_Cubic, time=0, value=1}
         {arriveTangent=0.0144, interpMode=RCIM_Cubic, leaveTangent=0.0144, tangentMode=RCTM_User, time=1, value=1.8}
        ]}}}))
   &DynamicMaterialParameters(
    Index 0 Param 1=&RandomRangeFloat(Minimum=-0.25, Maximum=0.25)
    Index 0 Param 2=1
    Index 0 Param 3=-0.8
    Index 0 Param 4=&FloatFromCurve(FloatCurve=DataInterface() {propertyValues {Curve {keys [
         {interpMode=RCIM_Cubic, tangentMode=RCTM_User, time=0, value=1}
         {arriveTangent=-0.6219, interpMode=RCIM_Cubic, leaveTangent=-0.6219, tangentMode=RCTM_User, time=1, value=0}
        ]}}})
    Module.Param0WriteEnabled=true
    Module.Param1WriteEnabled=true
    Module.Param2WriteEnabled=false
    Module.Param3WriteEnabled=false
   )
   &Color(Color=&ColorFromCurve(ColorCurve=DataInterface() {propertyValues {
       AlphaCurve {keys [<at 0.4167 1>,<at 1 0 RCIM_Cubic>]}
       BlueCurve {keys [<at 0.006 0>,<at 0.2174 0 RCIM_Cubic>,<at 0.5013 0 RCIM_Cubic>,<at 0.8696 0>]}
       GreenCurve {keys [
         <at 0.006 0.25>
         {arriveTangent=-0.488, interpMode=RCIM_Cubic, leaveTangent=-0.488, time=0.2174, value=0.0667}
         {interpMode=RCIM_Cubic, leaveTangent=-0.2544, time=0.5013, value=0.0083}
         <at 0.8696 0>
        ]}
       RedCurve {keys [
         <at 0.006 3>
         {arriveTangent=-5.856, interpMode=RCIM_Cubic, leaveTangent=-5.856, time=0.2174, value=0.8}
         {interpMode=RCIM_Cubic, leaveTangent=-3.0522, time=0.5013, value=0.1}
         <at 0.8696 0>
        ]}
      }}))
   /SetVariables(Particles.Color=&MultiplyLinearColor(
     Linear Color=$Particles.Color
     Scale Factor=&MakeFloat4FromVectorAndFloat(
      "Vector (RGB)"=&VectorFromFloat(Value=&Lerp_Float(A=0, B=1, Alpha=$User.WeakPointHitFade))
      "Float (Alpha)"=&Lerp_Float(A=0.8, B=1.5, Alpha=$User.WeakPointHitFade)
     )
    )) {isSetParameters=true}
  ]
  rendererClasses [&NiagaraRibbonRendererProperties]
  renderers [{rendererClass=&NiagaraRibbonRendererProperties}]
 }
 Trail_Smoke002 {
  emitterUpdateScript [
   &EmitterState(Life Cycle Mode=::NewEnumerator1, Loop Behavior=::NewEnumerator1, Loop Duration=0.27, UseLoopDelay=true)
   &FN_SpawnAnimTrail_v2()
  ]
  particleSpawnScript [
   &SampleSkeletalMesh(
    AdditionalYawPitchRoll=false
    Bone / Socket Index:NiagaraInt32=&Add_Integer(A=$Particles.SpawnGroup, B=0)
    Mesh Sampling Type:ENiagaraSkelSamplingModeFull=::NewEnumerator5
    Skeletal Mesh:NiagaraDataInterfaceSkeletalMesh=$User.SkeletalMesh
    Socket Sampling Mode:ENiagaraSocketSamplingMode=::NewEnumerator5
   )
   &FN_InitializeAnimTrailBoneIndecies(First Bone or Socket Index=1)
   &FN_InitializeAnimTrail(
    AnimTrailWidth=1.25
    Color=<rgba 0 0 0 3>
    Lifetime=0.225
    Position Offset:Vector3f=&Multiply_VectorByFloat(
     Vector=&NormalizeVector(
      Vector To Normalize=$Particles.SampleSkeletalMesh.SampledVelocity
      Safe Normalize=true
      Safe Fallback Threshold=0
      Safe Fallback Vector=<0 0 0.001>
     )
     Float=5
    )
   )
   &SystemLocation(MaskSpawnGroup=false)
   &AddVelocity(
    Rotation Coordinate Space=::Simulation
    Velocity=&RotateVectorByQuaternion(Vector To Rotate=<1 0 2>, Quaternion=$Particles.SampleSkeletalMesh.SampledRotation)
    Velocity Speed Scale=100
   )
   &CameraOffset(Camera Offset Amount=-100)
  ]
  particleUpdateScript [
   &FN_UpdateAnimTrail_v2()
   &ScaleRibbonWidth(Ribbon Width Scale=&FloatFromCurve(FloatCurve=DataInterface() {propertyValues {Curve {keys [{interpMode=RCIM_Cubic, tangentMode=RCTM_User, time=0, value=1}, {interpMode=RCIM_Cubic, tangentMode=RCTM_User, time=1, value=1.6}]}}})) {module=ScaleRibbonWidth001}
   &GravityForce(Gravity=<0 0 0>)
   &Drag(Drag=5, Use Linear Drag=true, Use Rotational Drag=true)
   &SolveForcesAndVelocity(Clamp Velocity=false, Limit Acceleration=false)
   &DynamicMaterialParameters(
    Index 0 Param 1=&RandomRangeFloat(Minimum=-0.25, Maximum=0.25)
    Index 0 Param 2=1
    Index 0 Param 3=-0.8
    Index 0 Param 4=&FloatFromCurve(FloatCurve=DataInterface() {propertyValues {Curve {keys [
         {interpMode=RCIM_Cubic, tangentMode=RCTM_User, time=0, value=1}
         {arriveTangent=-0.6219, interpMode=RCIM_Cubic, leaveTangent=-0.6219, tangentMode=RCTM_User, time=1, value=0}
        ]}}})
    Module.Param0WriteEnabled=true
    Module.Param1WriteEnabled=true
    Module.Param2WriteEnabled=false
    Module.Param3WriteEnabled=false
   )
   &Color(Color=&ColorFromCurve(ColorCurve=DataInterface() {propertyValues {
       AlphaCurve {keys [<at 0.2401 4 RCIM_Cubic>,<at 0.9527 0 RCIM_Cubic>]}
       BlueCurve {keys [<at 0.0224 0 RCIM_Cubic>, {arriveTangent=0.0106, interpMode=RCIM_Cubic, time=0.0745, value=0.0211}, <at 0.8817 0.0081>]}
       GreenCurve {keys [<at 0.0224 0.0833 RCIM_Cubic>, {arriveTangent=-0.0982, interpMode=RCIM_Cubic, time=0.0745, value=0.0211}, <at 0.8817 0.0081>]}
       RedCurve {keys [<at 0.0224 1 RCIM_Cubic>, {arriveTangent=-1.2933, interpMode=RCIM_Cubic, time=0.0745, value=0.05}, <at 0.8817 0.01>]}
      }}))
   /SetVariables(Particles.Color=&MultiplyLinearColor(
     Linear Color=$Particles.Color
     Scale Factor=&MakeFloat4FromVectorAndFloat(
      "Vector (RGB)"=&VectorFromFloat(Value=&Lerp_Float(A=0.1, B=1, Alpha=1))
      "Float (Alpha)"=&Lerp_Float(A=0.35, B=1, Alpha=1)
     )
    )) {isSetParameters=true}
  ]
  rendererClasses [&NiagaraRibbonRendererProperties]
  renderers [{rendererClass=&NiagaraRibbonRendererProperties}]
 }
}
summary {
 systemName=NS_Pickaxe_CupidEvil_Trail
 emitters [
  {emitterName=Embers, rendererClasses [&NiagaraSpriteRendererProperties]}
  {emitterName=Trail_Smoke002, rendererClasses [&NiagaraRibbonRendererProperties]}
  {emitterName=Trail_Fire, rendererClasses [&NiagaraRibbonRendererProperties]}
  {emitterName=Trail_Screen, rendererClasses [&NiagaraRibbonRendererProperties]}
  {emitterName=Trail_Fire001, rendererClasses [&NiagaraRibbonRendererProperties]}
 ]
 userVariables [
  User.bIsFrontEnd {type=NiagaraBool}
  User.bIsFrontEndPreview {type=NiagaraBool}
  User.RibbonIDOffset {type=NiagaraInt32}
  User.SkeletalMesh {
   type=NiagaraDataInterfaceSkeletalMesh
   defaultValue {
    struct=NiagaraExt_VariableValue_DataInterface
    value {dataInterface=&self:NiagaraDataInterfaceSkeletalMesh_1}
   }
  }
  User.WeakPointHitFade {type=NiagaraFloat, defaultValue=0, description=""}
 ]
}
