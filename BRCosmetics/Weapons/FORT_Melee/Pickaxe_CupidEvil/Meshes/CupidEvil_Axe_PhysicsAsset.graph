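# Physics asset for the CupidEvil axe, a cosmetic two-handed melee pickaxe, giving its mesh a simple collision shape.
# It holds a single body on the Melee_TwoHanded_Handle bone, made of one capsule running roughly along the bone's Z axis and offset along it from the bone origin.
# It has no constraints, so it defines no jointed or separately simulated parts.
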PhysicsAsset /BRCosmetics/Weapons/FORT_Melee/Pickaxe_CupidEvil/Meshes/CupidEvil_Axe_PhysicsAsset
bodies {Melee_TwoHanded_Handle [{shapeType=Capsule, length=51.9908, radius=13.4824, center=<0.4579 -1.0417 23.6733>, rotation=<rot -0.3098 -90 0.1931>}]}
constraints []
